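# Sound cue for dropping the Shockwave Launcher weapon.
# A Random node picks one of three drop wave variations with equal weight, feeding a Modulator at the root.
# It plays quietly in the loot-collect sound class with shared weapon-drop attenuation and references a weapon-drop concurrency set that, by its name, stops the oldest instance.
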SoundCue /Game/Sounds/Fort_Weapons/Launchers/Shockwave/Shockwave_Launcher_Drop_Cue
root=Modulator_0
nodes [
 {id=Modulator_0, children [Random_0]}
 {id=Random_0, children [WavePlayer_0,WavePlayer_2,WavePlayer_3], properties {weights [1,1,1]}}
 {id=WavePlayer_0, properties {soundWaveAssetPtr=&/Shockwave_Launcher_Drop_03}}
 {id=WavePlayer_2, properties {soundWaveAssetPtr=&/Shockwave_Launcher_Drop_01}}
 {id=WavePlayer_3, properties {soundWaveAssetPtr=&/Shockwave_Launcher_Drop_02}}
]
properties {
 attenuationSettings=&Fort_Weapon_Drop_Att
 maxDistance=3000
 soundClassObject=&Fort_Loot_Collect
 volumeMultiplier=0.3
 concurrencyOverrides {resolutionRule=PreventNew}
 concurrencySet [&Weapon_Drop_StopOldest]
}
