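# Sound cue for the Shockwave Launcher explosion, routed to the weapons sound class.
# A mixer layers a close branch and a distant branch, each with its own attenuation asset, so the blast's character changes with distance.
# Each branch picks one of three variations at random, with pitch modulation on close and volume modulation on distant.
# A concurrency override allows few simultaneous instances, stopping the oldest.
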SoundCue /Game/Sounds/Fort_Weapons/Launchers/Shockwave/Shockwave_Launcher_Explode_Cue
root=Mixer_2
nodes [
 {
  id=Attenuation_0
  children [Modulator_13]
  properties {attenuationSettings=/Game/Sounds/Attenuation/Fort_Shockwave_Launcher_Explo_Distant_Att}
 }
 {id=Attenuation_1, children [Modulator_0], properties {attenuationSettings=/Game/Sounds/Attenuation/Fort_Shockwave_Launcher_Explo_Close_Att}}
 {id=Mixer_2, children [Attenuation_1, Attenuation_0], properties {inputVolume [1.1,1]}}
 {id=Modulator_0, children [Random_1], properties {pitchMax=1.1, pitchMin=0.9}}
 {id=Modulator_13, children [Random_3], properties {volumeMax=1.1, volumeMin=1}}
 {id=Random_1, children [WavePlayer_6,WavePlayer_7,WavePlayer_8], properties {weights [1,1,1]}}
 {id=Random_3, children [WavePlayer_9, WavePlayer_10, WavePlayer_17], properties {weights [1,1,1]}}
 {id=WavePlayer_10, properties {soundWaveAssetPtr=&/Shockwave_Launcher_Explode_Distant_02}}
 {id=WavePlayer_17, properties {soundWaveAssetPtr=&/Shockwave_Launcher_Explode_Distant_03}}
 {id=WavePlayer_6, properties {soundWaveAssetPtr=&/Shockwave_Launcher_Explode_Close_01}}
 {id=WavePlayer_7, properties {soundWaveAssetPtr=&/Shockwave_Launcher_Explode_Close_02}}
 {id=WavePlayer_8, properties {soundWaveAssetPtr=&/Shockwave_Launcher_Explode_Close_03}}
 {id=WavePlayer_9, properties {soundWaveAssetPtr=&/Shockwave_Launcher_Explode_Distant_01}}
]
properties {
 bOverrideConcurrency=true
 maxDistance=13000
 soundClassObject=/Game/Sounds/FortSoundClassesAndModes/Fort_Weapons
 volumeMultiplier=0.8148
 attenuationOverrides {distanceAlgorithm=NaturalSound, falloffDistance=3800, lPFRadiusMax=3000, lPFRadiusMin=2000, attenuationShapeExtents {x=2200}}
 concurrencyOverrides {maxCount=2, resolutionRule=StopOldest}
}
