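# First-person fire sound cue for the Shockwave Launcher.
# A mixer layers a quieter generic grenade-fire wave, a random close-fire shot through a default modulator, and a boosted random barrel-rotate sound pitched down by a modulator.
# Output uses the first-person ranged-fire sound class, projectile attenuation and a first-person gunfire concurrency set.
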SoundCue /Game/Sounds/Fort_Weapons/Launchers/Shockwave/Shockwave_Launcher_Fire_1P_Cue
root=Mixer_0
nodes [
 {id=Mixer_0, children [WavePlayer_11, Modulator_15, Modulator_0], properties {inputVolume [0.5,1,1.4]}}
 {id=Modulator_0, children [Random_2], properties {pitchMax=0.9, pitchMin=0.9}}
 {id=Modulator_15, children [Random_1]}
 {id=Random_1, children [WavePlayer_3,WavePlayer_4,WavePlayer_5], properties {weights [1,1,1]}}
 {id=Random_2, children [WavePlayer_6,WavePlayer_7,WavePlayer_8,WavePlayer_9], properties {weights [1,1,1,1]}}
 {id=WavePlayer_11, properties {soundWaveAssetPtr=/Game/Sounds/Fort_Weapons/Launchers/weapon_grenade_fire_04}}
 {id=WavePlayer_3, properties {soundWaveAssetPtr=&/Shockwave_Launcher_Fire_Close_01}}
 {id=WavePlayer_4, properties {soundWaveAssetPtr=&/Shockwave_Launcher_Fire_Close_03}}
 {id=WavePlayer_5, properties {soundWaveAssetPtr=&/Shockwave_Launcher_Fire_Close_05}}
 {id=WavePlayer_6, properties {soundWaveAssetPtr=&/Shockwave_Launcher_BarrelRotate_01}}
 {id=WavePlayer_7, properties {soundWaveAssetPtr=&/Shockwave_Launcher_BarrelRotate_02}}
 {id=WavePlayer_8, properties {soundWaveAssetPtr=&/Shockwave_Launcher_BarrelRotate_03}}
 {id=WavePlayer_9, properties {soundWaveAssetPtr=&/Shockwave_Launcher_BarrelRotate_04}}
]
properties {
 attenuationSettings=&Fort_Weapon_Projectile_FirstPerson_Athena_Att
 maxDistance=2097152
 soundClassObject=&Fort_WeaponRanged_Fire_1P_PSM
 volumeMultiplier=0.55
 attenuationOverrides {falloffDistance=5000, attenuationShapeExtents {x=1000}}
 concurrencyOverrides {bLimitToOwner=true, maxCount=3, resolutionRule=StopOldest}
 concurrencySet [&Weapon_Gunfire_FirstPerson_Conc]
}
