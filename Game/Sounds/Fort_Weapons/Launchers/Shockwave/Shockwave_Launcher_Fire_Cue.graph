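# Sound cue for the Shockwave Launcher firing, built as distance-layered gunshot audio.
# A close layer blends a generic grenade-fire wave with random close variants under a third-person fire sound class and close attenuation.
# Randomized medium and distant layers, each modulated with its own attenuation, share a distant-fire sound class.
# Both branches mix at the root, governed by local and global gunfire concurrency sets.
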SoundCue /Game/Sounds/Fort_Weapons/Launchers/Shockwave/Shockwave_Launcher_Fire_Cue
root=Mixer_1
nodes [
 {
  id=Attenuation_10
  children [Modulator_14]
  properties {attenuationSettings=&Fort_Weapon_Projectile_Medium_Mono_Athena_Att}
 }
 {
  id=Attenuation_11
  children [Modulator_13]
  properties {attenuationSettings=&Fort_Weapon_Projectile_Distant_Mono_Athena_Att}
 }
 {id=Attenuation_9, children [Modulator_15], properties {attenuationSettings=&Fort_Weapon_Projectile_Close_Athena_Att}}
 {id=Mixer_1, children [SoundClass_7,SoundClass_6], properties {inputVolume [0.5,1]}}
 {id=Mixer_2, children [Attenuation_10, Attenuation_11], properties {inputVolume [0.8,0.3]}}
 {id=Mixer_4, children [WavePlayer_24, Random_0], properties {inputVolume [0.4,1]}}
 {id=Modulator_13, children [Random_2], properties {volumeMax=0.6, volumeMin=0.5}}
 {id=Modulator_14, children [Random_1], properties {pitchMax=0.8, pitchMin=0.7}}
 {id=Modulator_15, children [Mixer_4]}
 {id=Random_0, children [WavePlayer_0,WavePlayer_1,WavePlayer_2], properties {bShouldExcludeFromBranchCulling=true, weights [1,1,1]}}
 {id=Random_1, children [WavePlayer_3,WavePlayer_4,WavePlayer_5], properties {bShouldExcludeFromBranchCulling=true, weights [1,1,1]}}
 {id=Random_2, children [WavePlayer_6,WavePlayer_7,WavePlayer_8], properties {bShouldExcludeFromBranchCulling=true, weights [1,1,1]}}
 {id=SoundClass_6, children [Mixer_2], properties {soundClassOverride=&Fort_WeaponRanged_Fire_Distant}}
 {
  id=SoundClass_7
  children [Attenuation_9]
  properties {soundClassOverride=&Fort_WeaponRanged_Fire_3P_PSM}
 }
 {id=WavePlayer_0, properties {soundWaveAssetPtr=&/Shockwave_Launcher_Fire_Close_01}}
 {id=WavePlayer_1, properties {soundWaveAssetPtr=&/Shockwave_Launcher_Fire_Close_03}}
 {id=WavePlayer_2, properties {soundWaveAssetPtr=&/Shockwave_Launcher_Fire_Close_05}}
 {id=WavePlayer_24, properties {soundWaveAssetPtr=/Game/Sounds/Fort_Weapons/Launchers/weapon_grenade_fire_04}}
 {id=WavePlayer_3, properties {soundWaveAssetPtr=&/Shockwave_Launcher_Fire_Medium_01}}
 {id=WavePlayer_4, properties {soundWaveAssetPtr=&/Shockwave_Launcher_Fire_Medium_02}}
 {id=WavePlayer_5, properties {soundWaveAssetPtr=&/Shockwave_Launcher_Fire_Medium_03}}
 {id=WavePlayer_6, properties {soundWaveAssetPtr=&/Shockwave_Launcher_Fire_Distant_03}}
 {id=WavePlayer_7, properties {soundWaveAssetPtr=&/Shockwave_Launcher_Fire_Distant_01}}
 {id=WavePlayer_8, properties {soundWaveAssetPtr=&/Shockwave_Launcher_Fire_Distant_02}}
]
properties {
 maxDistance=27500
 volumeMultiplier=1
 attenuationOverrides {falloffDistance=5000, attenuationShapeExtents {x=1000}}
 concurrencyOverrides {bLimitToOwner=true, maxCount=3, resolutionRule=StopOldest}
 concurrencySet [&Weapon_Gunfire_Local_Conc, &Weapon_Gunfire_Global_Conc]
}
