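# Sound cue for the looping in-air sound of the Shockwave launcher, evidently played on the projectile in flight, in the Fort_Explosives sound class.
# One looping wave feeds a default-setting Modulator for random pitch and volume variation, then a strengthened Doppler node exaggerates pitch shift from motion.
# Inline attenuation overrides the referenced fuse attenuation asset with logarithmic falloff and HRTF spatialization.
# Overridden concurrency prevents new instances once the limit is reached.
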SoundCue /Game/Sounds/Fort_Weapons/Launchers/Shockwave/Shockwave_Launcher_InAirLoop_Cue
root=Doppler_0
nodes [
 {id=Doppler_0, children [Modulator_0], properties {dopplerIntensity=10}}
 {id=Modulator_0, children [WavePlayer_1]}
 {id=WavePlayer_1, properties {bLooping=true, soundWaveAssetPtr=&/Shockwave_Launcher_InAirLoop_01}}
]
properties {
 attenuationSettings=/Game/Sounds/Attenuation/Fort_Explosive_Fuse_Ambient_Att
 bOverrideAttenuation=true
 bOverrideConcurrency=true
 maxDistance=1700
 soundClassObject=/Game/Sounds/FortSoundClassesAndModes/Fort_Explosives
 volumeMultiplier=0.8
 attenuationOverrides {distanceAlgorithm=Logarithmic, falloffDistance=1200, spatializationAlgorithm=SPATIALIZATION_HRTF, attenuationShapeExtents {x=500}}
 concurrencyOverrides {resolutionRule=PreventNew}
}
