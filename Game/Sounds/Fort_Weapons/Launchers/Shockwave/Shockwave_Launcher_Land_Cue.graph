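# Sound cue for the Shockwave Launcher projectile landing, in the weapon projectile effects sound class.
# It layers a random Shockwave Grenade stick sound with a random launcher land sound at half volume through a mixer.
# A modulator adds slight pitch variation to the combined sound.
# Attenuation is overridden with HRTF spatialization and distance-based low-pass filtering, and the cue's overall volume is boosted.
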SoundCue /Game/Sounds/Fort_Weapons/Launchers/Shockwave/Shockwave_Launcher_Land_Cue
root=Modulator_0
nodes [
 {id=Mixer_0, children [Random_0,Random_1], properties {inputVolume [1,0.5]}}
 {id=Modulator_0, children [Mixer_0], properties {pitchMax=1.1, pitchMin=0.9}}
 {id=Random_0, children [WavePlayer_3,WavePlayer_4], properties {weights [1,1]}}
 {id=Random_1, children [WavePlayer_2,WavePlayer_5], properties {weights [1,1]}}
 {id=WavePlayer_2, properties {soundWaveAssetPtr=&/Shockwave_Launcher_Land_02}}
 {id=WavePlayer_3, properties {soundWaveAssetPtr=/Game/Sounds/Fort_GamePlay_Sounds/Explosives/ShockwaveGrenade/Grenade_Shockwave_Stick02}}
 {id=WavePlayer_4, properties {soundWaveAssetPtr=/Game/Sounds/Fort_GamePlay_Sounds/Explosives/ShockwaveGrenade/Grenade_Shockwave_Stick03}}
 {id=WavePlayer_5, properties {soundWaveAssetPtr=&/Shockwave_Launcher_Land_03}}
]
properties {
 attenuationSettings=/Game/Sounds/Attenuation/Fort_Grenade_Bounce_Att
 bOverrideAttenuation=true
 maxDistance=3000
 soundClassObject=&Fort_WeaponRanged_Projectile_Effects
 volumeMultiplier=2
 attenuationOverrides {
  bAttenuateWithLPF=true
  falloffDistance=2500
  lPFFrequencyAtMax=2000
  lPFRadiusMax=3000
  lPFRadiusMin=500
  spatializationAlgorithm=SPATIALIZATION_HRTF
  attenuationShapeExtents {x=500}
 }
}
