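# Sound cue for picking up the Shockwave Launcher weapon.
# A random node chooses one of three equally weighted pickup wave variations.
# It plays quietly in the loot collect sound class with shared weapon pickup attenuation.
# It references a shared pickup concurrency set and also carries single-instance override values that block new plays.
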SoundCue /Game/Sounds/Fort_Weapons/Launchers/Shockwave/Shockwave_Launcher_Pickup_Cue
root=Random_0
nodes [
 {id=Random_0, children [WavePlayer_3,WavePlayer_4,WavePlayer_7], properties {weights [1,1,1]}}
 {id=WavePlayer_3, properties {soundWaveAssetPtr=&/Shockwave_Launcher_Pickup_01}}
 {id=WavePlayer_4, properties {soundWaveAssetPtr=&/Shockwave_Launcher_Pickup_02}}
 {id=WavePlayer_7, properties {soundWaveAssetPtr=&/Shockwave_Launcher_Pickup_03}}
]
properties {
 attenuationSettings=&Fort_Weapon_Pickup_Att
 maxDistance=3500
 soundClassObject=&Fort_Loot_Collect
 volumeMultiplier=0.2769
 concurrencyOverrides {maxCount=1, resolutionRule=PreventNew}
 concurrencySet [&PickupMaximum_Oldest]
}
