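# Sound cue for the end of the Shockwave Launcher reload, played as quiet weapon foley with reload attenuation.
# It mixes a random pick of two equip sounds, run through a default modulator, with a second layer pitched down by a fixed amount.
# That second layer plays all three reload-insert waves at once through a nested mixer rather than choosing one.
# Concurrency is overridden to one instance, stopping the farthest first.
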SoundCue /Game/Sounds/Fort_Weapons/Launchers/Shockwave/Shockwave_Launcher_ReloadEnd_Cue
root=Mixer_1
nodes [
 {id=Mixer_0, children [WavePlayer_4,WavePlayer_5,WavePlayer_6], properties {inputVolume [1,1,1]}}
 {id=Mixer_1, children [Modulator_1,Modulator_0], properties {inputVolume [1,1]}}
 {id=Modulator_0, children [Mixer_0], properties {pitchMax=0.8, pitchMin=0.8}}
 {id=Modulator_1, children [Random_0]}
 {id=Random_0, children [WavePlayer_1,WavePlayer_3], properties {weights [1,1]}}
 {id=WavePlayer_1, properties {soundWaveAssetPtr=&/Shockwave_Launcher_Equip_01}}
 {id=WavePlayer_3, properties {soundWaveAssetPtr=&/Shockwave_Launcher_Equip_03}}
 {id=WavePlayer_4, properties {soundWaveAssetPtr=&/Shockwave_Launcher_ReloadInsert_01}}
 {id=WavePlayer_5, properties {soundWaveAssetPtr=&/Shockwave_Launcher_ReloadInsert_02}}
 {id=WavePlayer_6, properties {soundWaveAssetPtr=&/Shockwave_Launcher_ReloadInsert_03}}
]
properties {
 attenuationSettings=&Fort_Weapon_Reload_Att
 bOverrideConcurrency=true
 maxDistance=2250
 pitchMultiplier=1.1
 soundClassObject=&Fort_WeaponFoley
 volumeMultiplier=0.2
 attenuationOverrides {falloffDistance=5000, attenuationShapeExtents {x=1000}}
 concurrencyOverrides {maxCount=1, resolutionRule=StopFarthestThenPreventNew}
}
