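# Sound cue for the start of the Shockwave Launcher reload, played as weapon foley with reload attenuation.
# A mixer layers one of three random ReloadStart variants with one of three random ReloadInsert variants, the latter delayed by a randomized interval just under a second, pitched down and mixed louder.
# An outer modulator wraps the mix, and a concurrency override allows one instance, stopping the farthest first.
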SoundCue /Game/Sounds/Fort_Weapons/Launchers/Shockwave/Shockwave_Launcher_ReloadStart_Cue
root=Modulator_1
nodes [
 {id=Delay_0, children [Random_1], properties {delayMax=1, delayMin=0.7}}
 {id=Mixer_0, children [Random_0,Modulator_0], properties {inputVolume [1,1.4]}}
 {id=Modulator_0, children [Delay_0], properties {pitchMax=0.9, pitchMin=0.9}}
 {id=Modulator_1, children [Mixer_0]}
 {id=Random_0, children [WavePlayer_1,WavePlayer_2,WavePlayer_3], properties {weights [1,1,1]}}
 {id=Random_1, children [WavePlayer_4,WavePlayer_5,WavePlayer_6], properties {weights [1,1,1]}}
 {id=WavePlayer_1, properties {soundWaveAssetPtr=&/Shockwave_Launcher_ReloadStart_01}}
 {id=WavePlayer_2, properties {soundWaveAssetPtr=&/Shockwave_Launcher_ReloadStart_02}}
 {id=WavePlayer_3, properties {soundWaveAssetPtr=&/Shockwave_Launcher_ReloadStart_03}}
 {id=WavePlayer_4, properties {soundWaveAssetPtr=&/Shockwave_Launcher_ReloadInsert_01}}
 {id=WavePlayer_5, properties {soundWaveAssetPtr=&/Shockwave_Launcher_ReloadInsert_02}}
 {id=WavePlayer_6, properties {soundWaveAssetPtr=&/Shockwave_Launcher_ReloadInsert_03}}
]
properties {
 attenuationSettings=&Fort_Weapon_Reload_Att
 bOverrideConcurrency=true
 maxDistance=2250
 soundClassObject=&Fort_WeaponFoley
 volumeMultiplier=0.351
 attenuationOverrides {falloffDistance=5000, attenuationShapeExtents {x=1000}}
 concurrencyOverrides {maxCount=1, resolutionRule=StopFarthestThenPreventNew}
}
